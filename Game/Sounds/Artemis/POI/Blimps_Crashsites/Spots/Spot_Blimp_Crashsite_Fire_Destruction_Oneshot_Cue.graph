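# One-shot ambient Sound Cue, named as a fire and destruction spot for blimp crash sites, reusing ZeroPoint trapped-tower distant sounds.
# A quality-level root sends the first two levels to a random pick of seven waves and the last two to a four-wave subset.
# Each branch varies pitch and volume through a modulator.
# It plays quietly in the outdoor ambient sound class with wider one-shot attenuation.
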SoundCue /Game/Sounds/Artemis/POI/Blimps_Crashsites/Spots/Spot_Blimp_Crashsite_Fire_Destruction_Oneshot_Cue
root=QualityLevel_0
nodes [
 {id=Modulator_0, children [Random_3], properties {pitchMax=1.2, pitchMin=1, volumeMax=1.1, volumeMin=0.9}}
 {id=Modulator_1, children [Random_0], properties {pitchMax=1.2, pitchMin=1, volumeMax=1.1, volumeMin=0.8}}
 {id=QualityLevel_0, children [Modulator_1,Modulator_1,Modulator_0,Modulator_0]}
 {
  id=Random_0
  children [WavePlayer_0, WavePlayer_7, WavePlayer_8, WavePlayer_9, WavePlayer_10, WavePlayer_17, WavePlayer_18]
  properties {weights [1,1,1,1,1,1,1]}
 }
 {id=Random_3, children [WavePlayer_1,WavePlayer_2,WavePlayer_3,WavePlayer_4], properties {weights [1,1,1,1]}}
 {id=WavePlayer_0, properties {soundWaveAssetPtr=/Game/Sounds/Apollo/Biomes/ZeroPoint/AMB_SS_S16_ZP_Trapped_Tower_Dist_01}}
 {id=WavePlayer_1, properties {soundWaveAssetPtr=/Game/Sounds/Apollo/Biomes/ZeroPoint/AMB_SS_S16_ZP_Trapped_Tower_Dist_04}}
 {id=WavePlayer_10, properties {soundWaveAssetPtr=/Game/Sounds/Apollo/Biomes/ZeroPoint/AMB_SS_S16_ZP_Trapped_Tower_Dist_05}}
 {id=WavePlayer_17, properties {soundWaveAssetPtr=/Game/Sounds/Apollo/Biomes/ZeroPoint/AMB_SS_S16_ZP_Trapped_Tower_Dist_06}}
 {id=WavePlayer_18, properties {soundWaveAssetPtr=/Game/Sounds/Apollo/Biomes/ZeroPoint/AMB_SS_S16_ZP_Trapped_Tower_Dist_07}}
 {id=WavePlayer_2, properties {soundWaveAssetPtr=&AMB_SS_S16_ZP_Trapped_Tower_Dist_05}}
 {id=WavePlayer_3, properties {soundWaveAssetPtr=&AMB_SS_S16_ZP_Trapped_Tower_Dist_06}}
 {id=WavePlayer_4, properties {soundWaveAssetPtr=&AMB_SS_S16_ZP_Trapped_Tower_Dist_07}}
 {id=WavePlayer_7, properties {soundWaveAssetPtr=/Game/Sounds/Apollo/Biomes/ZeroPoint/AMB_SS_S16_ZP_Trapped_Tower_Dist_02}}
 {id=WavePlayer_8, properties {soundWaveAssetPtr=/Game/Sounds/Apollo/Biomes/ZeroPoint/AMB_SS_S16_ZP_Trapped_Tower_Dist_03}}
 {id=WavePlayer_9, properties {soundWaveAssetPtr=&AMB_SS_S16_ZP_Trapped_Tower_Dist_04}}
]
properties {
 attenuationSettings=/Game/Sounds/Attenuation/Fort_TileAmbient_OneShot_Wider_Att
 maxDistance=8000
 soundClassObject=&Fort_Ambient_Outdoors_Athena
 volumeMultiplier=0.301
 preEffectBusSends [{}]
}
