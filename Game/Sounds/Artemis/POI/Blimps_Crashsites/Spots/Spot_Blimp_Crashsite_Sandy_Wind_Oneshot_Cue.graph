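# Sound cue for one-shot sandy wind ambience at blimp crashsite spots, reusing Tower Aftermath wind recordings.
# A quality-level root sends its first two inputs to a random pick of twelve waves and the last two to a four-wave subset.
# Each random pool is equally weighted and feeds a modulator for slight pitch and volume variation.
# It plays at roughly half volume in the outdoor ambient class, with wide one-shot attenuation.
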SoundCue /Game/Sounds/Artemis/POI/Blimps_Crashsites/Spots/Spot_Blimp_Crashsite_Sandy_Wind_Oneshot_Cue
root=QualityLevel_0
nodes [
 {id=Modulator_0, children [Random_2], properties {pitchMax=1.1, pitchMin=1, volumeMax=1.1, volumeMin=0.9}}
 {id=Modulator_1, children [Random_1], properties {pitchMax=1.1, pitchMin=1, volumeMax=1.1, volumeMin=0.9}}
 {id=QualityLevel_0, children [Modulator_1,Modulator_1,Modulator_0,Modulator_0]}
 {
  id=Random_1
  children [
   WavePlayer_24
   WavePlayer_25
   WavePlayer_26
   WavePlayer_27
   WavePlayer_28
   WavePlayer_29
   WavePlayer_30
   WavePlayer_31
   WavePlayer_32
   WavePlayer_33
   WavePlayer_38
   WavePlayer_39
  ]
  properties {weights [1,1,1,1,1,1,1,1,1,1,1,1]}
 }
 {id=Random_2, children [WavePlayer_40, WavePlayer_41, WavePlayer_42, WavePlayer_43], properties {weights [1,1,1,1]}}
 {id=WavePlayer_24, properties {soundWaveAssetPtr=/Game/Sounds/Apollo/Biomes/Tower_Aftermath/AMB_Tower_AM_Sandy_Wind_Oneshot_05}}
 {id=WavePlayer_25, properties {soundWaveAssetPtr=/Game/Sounds/Apollo/Biomes/Tower_Aftermath/AMB_Tower_AM_Sandy_Wind_Oneshot_10}}
 {id=WavePlayer_26, properties {soundWaveAssetPtr=/Game/Sounds/Apollo/Biomes/Tower_Aftermath/AMB_Tower_AM_Sandy_Wind_Oneshot_11}}
 {id=WavePlayer_27, properties {soundWaveAssetPtr=/Game/Sounds/Apollo/Biomes/Tower_Aftermath/AMB_Tower_AM_Sandy_Wind_Oneshot_12}}
 {id=WavePlayer_28, properties {soundWaveAssetPtr=/Game/Sounds/Apollo/Biomes/Tower_Aftermath/AMB_Tower_AM_Sandy_Wind_Oneshot_01}}
 {id=WavePlayer_29, properties {soundWaveAssetPtr=/Game/Sounds/Apollo/Biomes/Tower_Aftermath/AMB_Tower_AM_Sandy_Wind_Oneshot_02}}
 {id=WavePlayer_30, properties {soundWaveAssetPtr=/Game/Sounds/Apollo/Biomes/Tower_Aftermath/AMB_Tower_AM_Sandy_Wind_Oneshot_03}}
 {id=WavePlayer_31, properties {soundWaveAssetPtr=/Game/Sounds/Apollo/Biomes/Tower_Aftermath/AMB_Tower_AM_Sandy_Wind_Oneshot_04}}
 {id=WavePlayer_32, properties {soundWaveAssetPtr=/Game/Sounds/Apollo/Biomes/Tower_Aftermath/AMB_Tower_AM_Sandy_Wind_Oneshot_06}}
 {id=WavePlayer_33, properties {soundWaveAssetPtr=/Game/Sounds/Apollo/Biomes/Tower_Aftermath/AMB_Tower_AM_Sandy_Wind_Oneshot_07}}
 {id=WavePlayer_38, properties {soundWaveAssetPtr=/Game/Sounds/Apollo/Biomes/Tower_Aftermath/AMB_Tower_AM_Sandy_Wind_Oneshot_08}}
 {id=WavePlayer_39, properties {soundWaveAssetPtr=/Game/Sounds/Apollo/Biomes/Tower_Aftermath/AMB_Tower_AM_Sandy_Wind_Oneshot_09}}
 {id=WavePlayer_40, properties {soundWaveAssetPtr=&AMB_Tower_AM_Sandy_Wind_Oneshot_03}}
 {id=WavePlayer_41, properties {soundWaveAssetPtr=&AMB_Tower_AM_Sandy_Wind_Oneshot_06}}
 {id=WavePlayer_42, properties {soundWaveAssetPtr=&AMB_Tower_AM_Sandy_Wind_Oneshot_09}}
 {id=WavePlayer_43, properties {soundWaveAssetPtr=&AMB_Tower_AM_Sandy_Wind_Oneshot_01}}
]
properties {
 attenuationSettings=/Game/Sounds/Attenuation/Fort_TileAmbient_OneShot_Wider_Att
 maxDistance=8000
 soundClassObject=&Fort_Ambient_Outdoors_Athena
 volumeMultiplier=0.501
 preEffectBusSends [{}]
}
