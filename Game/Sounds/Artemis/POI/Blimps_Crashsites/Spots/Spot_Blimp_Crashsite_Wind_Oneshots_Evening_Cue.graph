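# Sound cue for evening wind gust one-shots at blimp crash site ambient spots, reusing Tower Aftermath gust waves.
# A quality level root sends its first two slots to an equally weighted random pick of eleven gusts and its last two to a four-gust subset.
# Each branch randomizes pitch and volume through its own modulator, with a wider volume range on the smaller set.
# Playback uses wide one-shot attenuation and the outdoor ambient sound class.
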SoundCue /Game/Sounds/Artemis/POI/Blimps_Crashsites/Spots/Spot_Blimp_Crashsite_Wind_Oneshots_Evening_Cue
root=QualityLevel_0
nodes [
 {id=Modulator_0, children [Random_3], properties {pitchMax=1.1, pitchMin=0.7, volumeMax=1.1, volumeMin=0.5}}
 {id=Modulator_1, children [Random_2], properties {pitchMax=1.1, pitchMin=0.7, volumeMax=1.1, volumeMin=0.9}}
 {id=QualityLevel_0, children [Modulator_1,Modulator_1,Modulator_0,Modulator_0]}
 {
  id=Random_2
  children [
   WavePlayer_0
   WavePlayer_1
   WavePlayer_2
   WavePlayer_3
   WavePlayer_4
   WavePlayer_5
   WavePlayer_6
   WavePlayer_7
   WavePlayer_8
   WavePlayer_9
   WavePlayer_10
  ]
  properties {weights [1,1,1,1,1,1,1,1,1,1,1]}
 }
 {id=Random_3, children [WavePlayer_26, WavePlayer_27, WavePlayer_28, WavePlayer_29], properties {weights [1,1,1,1]}}
 {id=WavePlayer_0, properties {soundWaveAssetPtr=/Game/Sounds/Apollo/Biomes/Tower_Aftermath/AMB_Tower_AM_WindGust_Evening_10}}
 {id=WavePlayer_1, properties {soundWaveAssetPtr=/Game/Sounds/Apollo/Biomes/Tower_Aftermath/AMB_Tower_AM_WindGust_Evening_11}}
 {id=WavePlayer_10, properties {soundWaveAssetPtr=/Game/Sounds/Apollo/Biomes/Tower_Aftermath/AMB_Tower_AM_WindGust_Evening_09}}
 {id=WavePlayer_2, properties {soundWaveAssetPtr=/Game/Sounds/Apollo/Biomes/Tower_Aftermath/AMB_Tower_AM_WindGust_Evening_01}}
 {id=WavePlayer_26, properties {soundWaveAssetPtr=&AMB_Tower_AM_WindGust_Evening_10}}
 {id=WavePlayer_27, properties {soundWaveAssetPtr=&AMB_Tower_AM_WindGust_Evening_11}}
 {id=WavePlayer_28, properties {soundWaveAssetPtr=&AMB_Tower_AM_WindGust_Evening_01}}
 {id=WavePlayer_29, properties {soundWaveAssetPtr=/Game/Sounds/Apollo/Biomes/Tower_Aftermath/AMB_Tower_AM_WindGust_Evening_02}}
 {id=WavePlayer_3, properties {soundWaveAssetPtr=&AMB_Tower_AM_WindGust_Evening_02}}
 {id=WavePlayer_4, properties {soundWaveAssetPtr=/Game/Sounds/Apollo/Biomes/Tower_Aftermath/AMB_Tower_AM_WindGust_Evening_03}}
 {id=WavePlayer_5, properties {soundWaveAssetPtr=/Game/Sounds/Apollo/Biomes/Tower_Aftermath/AMB_Tower_AM_WindGust_Evening_04}}
 {id=WavePlayer_6, properties {soundWaveAssetPtr=/Game/Sounds/Apollo/Biomes/Tower_Aftermath/AMB_Tower_AM_WindGust_Evening_05}}
 {id=WavePlayer_7, properties {soundWaveAssetPtr=/Game/Sounds/Apollo/Biomes/Tower_Aftermath/AMB_Tower_AM_WindGust_Evening_06}}
 {id=WavePlayer_8, properties {soundWaveAssetPtr=/Game/Sounds/Apollo/Biomes/Tower_Aftermath/AMB_Tower_AM_WindGust_Evening_07}}
 {id=WavePlayer_9, properties {soundWaveAssetPtr=/Game/Sounds/Apollo/Biomes/Tower_Aftermath/AMB_Tower_AM_WindGust_Evening_08}}
]
properties {
 attenuationSettings=/Game/Sounds/Attenuation/Fort_TileAmbient_OneShot_Wider_Att
 maxDistance=8000
 soundClassObject=&Fort_Ambient_Outdoors_Athena
 volumeMultiplier=0.501
 preEffectBusSends [{}]
}
