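# Night wind-gust one-shot ambience cue for blimp crash site spots, reusing Tower Aftermath wind gust waves.
# A quality-level switch sends its first two levels to a random pick from eleven gusts with slight pitch and volume modulation.
# The last two levels use a four-gust subset with wider modulation.
# It plays in the outdoor ambient sound class with wide one-shot tile-ambient attenuation.
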SoundCue /Game/Sounds/Artemis/POI/Blimps_Crashsites/Spots/Spot_Blimp_Crashsite_Wind_Oneshots_Night_Cue
root=QualityLevel_0
nodes [
 {id=Modulator_0, children [Random_1], properties {pitchMax=1.1, pitchMin=0.7, volumeMax=1.1, volumeMin=0.5}}
 {id=Modulator_1, children [Random_0], properties {pitchMax=1.2, pitchMin=1, volumeMax=1.1, volumeMin=0.9}}
 {id=QualityLevel_0, children [Modulator_1,Modulator_1,Modulator_0,Modulator_0]}
 {
  id=Random_0
  children [
   WavePlayer_11
   WavePlayer_12
   WavePlayer_13
   WavePlayer_14
   WavePlayer_15
   WavePlayer_16
   WavePlayer_17
   WavePlayer_18
   WavePlayer_19
   WavePlayer_20
   WavePlayer_21
  ]
  properties {weights [1,1,1,1,1,1,1,1,1,1,1]}
 }
 {id=Random_1, children [WavePlayer_22, WavePlayer_23, WavePlayer_24, WavePlayer_25], properties {weights [1,1,1,1]}}
 {id=WavePlayer_11, properties {soundWaveAssetPtr=/Game/Sounds/Apollo/Biomes/Tower_Aftermath/AMB_Tower_AM_WindGust_Night_09}}
 {id=WavePlayer_12, properties {soundWaveAssetPtr=/Game/Sounds/Apollo/Biomes/Tower_Aftermath/AMB_Tower_AM_WindGust_Night_10}}
 {id=WavePlayer_13, properties {soundWaveAssetPtr=/Game/Sounds/Apollo/Biomes/Tower_Aftermath/AMB_Tower_AM_WindGust_Night_11}}
 {id=WavePlayer_14, properties {soundWaveAssetPtr=/Game/Sounds/Apollo/Biomes/Tower_Aftermath/AMB_Tower_AM_WindGust_Night_01}}
 {id=WavePlayer_15, properties {soundWaveAssetPtr=/Game/Sounds/Apollo/Biomes/Tower_Aftermath/AMB_Tower_AM_WindGust_Night_02}}
 {id=WavePlayer_16, properties {soundWaveAssetPtr=/Game/Sounds/Apollo/Biomes/Tower_Aftermath/AMB_Tower_AM_WindGust_Night_03}}
 {id=WavePlayer_17, properties {soundWaveAssetPtr=/Game/Sounds/Apollo/Biomes/Tower_Aftermath/AMB_Tower_AM_WindGust_Night_04}}
 {id=WavePlayer_18, properties {soundWaveAssetPtr=/Game/Sounds/Apollo/Biomes/Tower_Aftermath/AMB_Tower_AM_WindGust_Night_05}}
 {id=WavePlayer_19, properties {soundWaveAssetPtr=/Game/Sounds/Apollo/Biomes/Tower_Aftermath/AMB_Tower_AM_WindGust_Night_06}}
 {id=WavePlayer_20, properties {soundWaveAssetPtr=/Game/Sounds/Apollo/Biomes/Tower_Aftermath/AMB_Tower_AM_WindGust_Night_07}}
 {id=WavePlayer_21, properties {soundWaveAssetPtr=/Game/Sounds/Apollo/Biomes/Tower_Aftermath/AMB_Tower_AM_WindGust_Night_08}}
 {id=WavePlayer_22, properties {soundWaveAssetPtr=&AMB_Tower_AM_WindGust_Night_01}}
 {id=WavePlayer_23, properties {soundWaveAssetPtr=&AMB_Tower_AM_WindGust_Night_11}}
 {id=WavePlayer_24, properties {soundWaveAssetPtr=&AMB_Tower_AM_WindGust_Night_10}}
 {id=WavePlayer_25, properties {soundWaveAssetPtr=&AMB_Tower_AM_WindGust_Night_02}}
]
properties {
 attenuationSettings=/Game/Sounds/Attenuation/Fort_TileAmbient_OneShot_Wider_Att
 maxDistance=8000
 soundClassObject=&Fort_Ambient_Outdoors_Athena
 volumeMultiplier=0.702
 preEffectBusSends [{}]
}
